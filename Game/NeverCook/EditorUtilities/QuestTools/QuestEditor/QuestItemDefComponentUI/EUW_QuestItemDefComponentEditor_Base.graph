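# Editor utility widget base for editing a quest item definition component in the Quest Editor tool, built on FortQuestItemDefComponentEditor.
# A dark header bar shows the component name with a small Copy button.
# Below it, a properties detail view shows all component properties except ObjectivesExtensionData.
# The Blueprint graph holds only unwired event stubs, and the name, button and detail view bind to widgets inherited from the parent class.
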EditorUtilityWidgetBlueprint /Game/NeverCook/EditorUtilities/QuestTools/QuestEditor/QuestItemDefComponentUI/EUW_QuestItemDefComponentEditor_Base
graphs [EventGraph {nodes [
   Event PreConstruct()
   Event Construct()
   Event Tick()
  ]}]
variables []
widgets {
 info {
  inheritedWidgetCount=3
  parentClass=/Script/FortniteEditor.FortQuestItemDefComponentEditor
  rootWidgetClass=&HorizontalBox
  widgetCount=11
 }
 widgets [
  HorizontalBox_0 {widgetClassPath=&HorizontalBox, widget=HorizontalBox_0}
  SizeBox_1 {
   widgetClassPath=&SizeBox
   parent=HorizontalBox_0
   widget=SizeBox_1
   properties {bOverride_WidthOverride=true, widthOverride=800}
   slot {class=HorizontalBoxSlot}
  }
  VerticalBox {widgetClassPath=&VerticalBox, parent=SizeBox_1, widget=VerticalBox, slot {class=SizeBoxSlot}}
  Border_0 {
   widgetClassPath=&Border
   parent=VerticalBox
   widget=Border_0
   properties {brushColor=<rgba 0.0284 0.0284 0.0284 1>, padding=<ltrb 4 6 4 6>}
   slot {class=VerticalBoxSlot, properties {verticalAlignment=VAlign_Center}}
  }
  HorizontalBox_1 {
   widgetClassPath=&HorizontalBox
   parent=Border_0
   widget=HorizontalBox_1
   slot {class=BorderSlot, properties {padding=<ltrb 4 6 4 6>}}
  }
  Text_ComponentName {
   widgetClassPath=&CommonTextBlock
   bInherited=true
   parent=HorizontalBox_1
   widget=Text_ComponentName
   properties {
    text=Component Name
    colorAndOpacity {colorUseRule=UseColor_Specified, specifiedColor=<rgba 0.9671 1 0.9899 1>}
    font {
     bForceMonospaced=false
     bMaterialIsStencil=false
     fontMaterial=null
     fontObject=&NotoSans
     letterSpacing=0
     monospacedWidth=1
     size=8
     skewAmount=0
     typefaceFontName=NotoSans-Regular
     outlineSettings {
      bApplyOutlineToDropShadows=false
      bMiteredCorners=false
      bSeparateFillAlpha=false
      outlineMaterial=null
      outlineSize=0
      outlineColor=<rgba 0 0 0 1>
     }
    }
    shadowOffset=<0 0>
   }
   slot {class=HorizontalBoxSlot, properties {horizontalAlignment=HAlign_Left, verticalAlignment=VAlign_Center, size {sizeRule=Fill, value=1}}}
  }
  Button_Copy {
   widgetClassPath=/Game/NeverCook/EditorUtilities/QuestTools/QuestEditor/QuestEditorButtons/QuestEditorButton
   bInherited=true
   bIsVariable=true
   parent=HorizontalBox_1
   widget=Button_Copy
   properties {
    style=/Game/NeverCook/EditorUtilities/QuestTools/QuestEditor/QuestEditorButtons/Styles/QuestEditorButtonStyle_Small.QuestEditorButtonStyle_Small_C
    text=Copy
    font {
     bForceMonospaced=false
     bMaterialIsStencil=false
     fontMaterial=null
     fontObject=&NotoSans
     letterSpacing=0
     monospacedWidth=1
     size=9.75
     skewAmount=0
     typefaceFontName=NotoSans-Regular
     outlineSettings {
      bApplyOutlineToDropShadows=false
      bMiteredCorners=false
      bSeparateFillAlpha=false
      outlineMaterial=null
      outlineSize=0
      outlineColor=<rgba 0 0 0 1>
     }
    }
   }
   slot {class=HorizontalBoxSlot}
  }
  Spacer_58 {widgetClassPath=&Spacer, parent=VerticalBox, widget=Spacer_58, slot {class=VerticalBoxSlot}}
  SizeBox_0 {
   widgetClassPath=&SizeBox
   parent=VerticalBox
   widget=SizeBox_0
   properties {bOverride_MinDesiredWidth=true, minDesiredWidth=600}
   slot {class=VerticalBoxSlot}
  }
  Border {
   widgetClassPath=&Border
   parent=SizeBox_0
   widget=Border
   properties {brushColor=<rgba 0.0176 0.0176 0.0176 1>, padding=<ltrb 0 0 0 0>}
   slot {class=SizeBoxSlot}
  }
  DetailView_ComponentProperties {
   widgetClassPath=/Script/PropertiesDetailView
   bInherited=true
   bIsVariable=true
   parent=Border
   widget=DetailView_ComponentProperties
   properties {
    bShowAllProperties=true
    visibility=SelfHitTestInvisible
    propertiesToExclude [ObjectivesExtensionData]
    propertyColumns [{fillCoefficient=1, propertiesDisplayData []}]
   }
   slot {class=BorderSlot, properties {padding=<ltrb 0 0 0 0>}}
  }
 ]
}
