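# Editor utility widget that serves as a map-name entry for a combo box in the quest item definition component UI of the quest editor tools.
# A horizontal box holds a collapsed "Map Name" label and a text block that fills the remaining width.
# InitMapEntry takes a map Name, converts it to text and writes it into that text block.
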EditorUtilityWidgetBlueprint /Game/NeverCook/EditorUtilities/QuestTools/QuestEditor/QuestItemDefComponentUI/EditorUtilityWidgets/ComboBoxMapNameContent
graphs [
 EventGraph {nodes [
   Event PreConstruct()
   Event Construct()
   Event Tick()
  ]}
 InitMapEntry {nodes [
   n1 = InitMapEntry!() -> (then, InMapName:Name) {is=title K2Node_FunctionEntry}
   > Set Text(Text="To Text (Name)"!(InName:Name=@n1) -> (ReturnValue:Text), self=Get Text_MapName() -> (Text_MapName:Common Text Object Reference))
  ]}
]
variables []
widgets {
 info {rootWidgetClass=&HorizontalBox, widgetCount=3}
 widgets [
  HorizontalBox_93 {widgetClassPath=&HorizontalBox, widget=HorizontalBox_93}
  Text_MapLabel {
   widgetClassPath=&CommonTextBlock
   parent=HorizontalBox_93
   widget=Text_MapLabel
   properties {
    text=Map Name
    visibility=Collapsed
    font {
     bForceMonospaced=false
     bMaterialIsStencil=false
     fontMaterial=null
     fontObject=&NotoSans
     letterSpacing=0
     monospacedWidth=1
     size=8
     skewAmount=0
     typefaceFontName=NotoSans-Regular
     outlineSettings {
      bApplyOutlineToDropShadows=false
      bMiteredCorners=false
      bSeparateFillAlpha=false
      outlineMaterial=null
      outlineSize=0
      outlineColor=<rgba 0 0 0 1>
     }
    }
    shadowOffset=<0 0>
   }
   slot {class=HorizontalBoxSlot, properties {verticalAlignment=VAlign_Center, padding=<ltrb 0 0 10 0>}}
  }
  Text_MapName {
   widgetClassPath=&CommonTextBlock
   bIsVariable=true
   parent=HorizontalBox_93
   widget=Text_MapName
   properties {
    text=Text Block
    font {
     bForceMonospaced=false
     bMaterialIsStencil=false
     fontMaterial=null
     fontObject=&NotoSans
     letterSpacing=0
     monospacedWidth=1
     size=8
     skewAmount=0
     typefaceFontName=NotoSans-Regular
     outlineSettings {
      bApplyOutlineToDropShadows=false
      bMiteredCorners=false
      bSeparateFillAlpha=false
      outlineMaterial=null
      outlineSize=0
      outlineColor=<rgba 0 0 0 1>
     }
    }
    shadowOffset=<0 0>
   }
   slot {class=HorizontalBoxSlot, properties {verticalAlignment=VAlign_Center, size {sizeRule=Fill, value=1}}}
  }
 ]
}
