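# Editor utility widget for the Fortnite quest editor that reports whether a quest's map marker tags have been found on the map.
# A vertical stack of status rows covers all found, missing exact match but with children, awaiting nightly script collection, and failed collection, plus an expandable Additional Info section.
# The Blueprint is layout only, with empty events; key boxes and text blocks bind to inherited members of the native MapMarkerTagsStatusWidget parent.
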EditorUtilityWidgetBlueprint /Game/NeverCook/EditorUtilities/QuestTools/QuestEditor/QuestItemDefComponentUI/MapMarkerStatus
graphs [EventGraph {nodes [
   Event PreConstruct()
   Event Construct()
   Event Tick()
  ]}]
variables []
widgets {
 info {
  inheritedWidgetCount=9
  namedSlotCount=2
  parentClass=/Script/FortniteEditor.MapMarkerTagsStatusWidget
  rootWidgetClass=&Border
  widgetCount=30
 }
 widgets [
  Border_0 {widgetClassPath=&Border, widget=Border_0, properties {brushColor=<rgba 0.0284 0.0284 0.0284 1>, padding=<ltrb 0 0 0 0>}}
  VerticalBox_25 {
   widgetClassPath=&VerticalBox
   parent=Border_0
   widget=VerticalBox_25
   slot {class=BorderSlot, properties {padding=<ltrb 0 0 0 0>}}
  }
  Box_TagsOk {
   widgetClassPath=&HorizontalBox
   bInherited=true
   parent=VerticalBox_25
   widget=Box_TagsOk
   slot {class=VerticalBoxSlot, properties {padding=<ltrb 0 0 0 2>}}
  }
  Border {
   widgetClassPath=&Border
   parent=Box_TagsOk
   widget=Border
   properties {brushColor=<rgba 0.0469 0.0469 0.0469 1>}
   slot {class=HorizontalBoxSlot, properties {padding=<ltrb 2 0 0 0>, size {sizeRule=Fill, value=1}}}
  }
  HorizontalBox_1 {widgetClassPath=&HorizontalBox, parent=Border, widget=HorizontalBox_1, slot {class=BorderSlot}}
  Image_TagsOk {
   widgetClassPath=&Image
   bIsVariable=true
   parent=HorizontalBox_1
   widget=Image_TagsOk
   properties {brush {
     bIsDynamicallyLoaded=false
     drawAs=Image
     imageType=NoImage
     mirroring=NoMirror
     resourceName=None
     resourceObject=/Game/Athena/HUD/Bacchus/HUDLayoutTool/HLT_Icons/T-HLT-CheckMark-32
     tiling=NoTile
     imageSize=<16 16>
     margin=<ltrb 0 0 0 0>
     outlineSettings {
      bUseBrushTransparency=false
      roundingType=HalfHeightRadius
      width=0
      color {colorUseRule=UseColor_Specified, specifiedColor=<rgba 0 0 0 0>}
      cornerRadii=<0 0 0 0>
     }
     tintColor {colorUseRule=UseColor_Specified, specifiedColor=<rgba 0.1173 1 0.0432 1>}
     uVRegion {bIsValid=false, max=<0 0>, min=<0 0>}
    }}
   slot {class=HorizontalBoxSlot, properties {horizontalAlignment=HAlign_Center, verticalAlignment=VAlign_Center}}
  }
  Text_TagsOk {
   widgetClassPath=&TextBlock
   parent=HorizontalBox_1
   widget=Text_TagsOk
   properties {
    text=All tag locations have been found on the map.
    font {
     bForceMonospaced=false
     bMaterialIsStencil=false
     fontMaterial=null
     fontObject=&NotoSans
     letterSpacing=0
     monospacedWidth=1
     size=8
     skewAmount=0
     typefaceFontName=NotoSans-Regular
     outlineSettings {
      bApplyOutlineToDropShadows=false
      bMiteredCorners=false
      bSeparateFillAlpha=false
      outlineMaterial=null
      outlineSize=0
      outlineColor=<rgba 0 0 0 1>
     }
    }
   }
   slot {class=HorizontalBoxSlot, properties {padding=<ltrb 4 2 4 2>}}
  }
  Box_FailedTagsWithChildrenWarning {
   widgetClassPath=&HorizontalBox
   bInherited=true
   parent=VerticalBox_25
   widget=Box_FailedTagsWithChildrenWarning
   slot {class=VerticalBoxSlot, properties {padding=<ltrb 0 0 0 2>}}
  }
  Border_4 {
   widgetClassPath=&Border
   parent=Box_FailedTagsWithChildrenWarning
   widget=Border_4
   properties {brushColor=<rgba 0.0469 0.0469 0.0469 1>}
   slot {class=HorizontalBoxSlot, properties {padding=<ltrb 2 0 0 0>, size {sizeRule=Fill, value=1}}}
  }
  VerticalBox_2 {widgetClassPath=&VerticalBox, parent=Border_4, widget=VerticalBox_2, slot {class=BorderSlot}}
  HorizontalBox_3 {widgetClassPath=&HorizontalBox, parent=VerticalBox_2, widget=HorizontalBox_3, slot {class=VerticalBoxSlot}}
  Image_RedWarning_1 {
   widgetClassPath=&Image
   bIsVariable=true
   parent=HorizontalBox_3
   widget=Image_RedWarning_1
   properties {brush {
     bIsDynamicallyLoaded=false
     drawAs=Image
     imageType=NoImage
     mirroring=NoMirror
     resourceName=None
     resourceObject=/Game/UI/Foundation/Shell/Textures/T-Icon-Warning-64
     tiling=NoTile
     imageSize=<16 16>
     margin=<ltrb 0 0 0 0>
     outlineSettings {
      bUseBrushTransparency=false
      roundingType=HalfHeightRadius
      width=0
      color {colorUseRule=UseColor_Specified, specifiedColor=<rgba 0 0 0 0>}
      cornerRadii=<0 0 0 0>
     }
     tintColor {colorUseRule=UseColor_Specified, specifiedColor=<rgba 1 0.9347 0.088 1>}
     uVRegion {bIsValid=false, max=<0 0>, min=<0 0>}
    }}
   slot {class=HorizontalBoxSlot, properties {horizontalAlignment=HAlign_Center, verticalAlignment=VAlign_Center}}
  }
  Text_FailedTagsWithChildrenWarning {
   widgetClassPath=&TextBlock
   bInherited=true
   parent=HorizontalBox_3
   widget=Text_FailedTagsWithChildrenWarning
   properties {
    text=Tag failed to find an exact match but has %d children that will be shown.
    font {
     bForceMonospaced=false
     bMaterialIsStencil=false
     fontMaterial=null
     fontObject=&NotoSans
     letterSpacing=0
     monospacedWidth=1
     size=8
     skewAmount=0
     typefaceFontName=NotoSans-Regular
     outlineSettings {
      bApplyOutlineToDropShadows=false
      bMiteredCorners=false
      bSeparateFillAlpha=false
      outlineMaterial=null
      outlineSize=0
      outlineColor=<rgba 0 0 0 1>
     }
    }
   }
   slot {class=HorizontalBoxSlot, properties {padding=<ltrb 4 2 4 2>}}
  }
  Box_PendingTagsWarning {
   widgetClassPath=&HorizontalBox
   bInherited=true
   parent=VerticalBox_25
   widget=Box_PendingTagsWarning
   slot {class=VerticalBoxSlot, properties {padding=<ltrb 0 0 0 2>}}
  }
  Border_3 {
   widgetClassPath=&Border
   parent=Box_PendingTagsWarning
   widget=Border_3
   properties {brushColor=<rgba 0.0469 0.0469 0.0469 1>}
   slot {class=HorizontalBoxSlot, properties {padding=<ltrb 2 0 0 0>, size {sizeRule=Fill, value=1}}}
  }
  VerticalBox_0 {widgetClassPath=&VerticalBox, parent=Border_3, widget=VerticalBox_0, slot {class=BorderSlot}}
  HorizontalBox_0 {
   widgetClassPath=&HorizontalBox
   parent=VerticalBox_0
   widget=HorizontalBox_0
   slot {class=VerticalBoxSlot, properties {size {sizeRule=Fill, value=1}}}
  }
  Image_YellowWarning {
   widgetClassPath=&Image
   bIsVariable=true
   parent=HorizontalBox_0
   widget=Image_YellowWarning
   properties {brush {
     bIsDynamicallyLoaded=false
     drawAs=Image
     imageType=NoImage
     mirroring=NoMirror
     resourceName=None
     resourceObject=&T-Icon-Warning-64
     tiling=NoTile
     imageSize=<16 16>
     margin=<ltrb 0 0 0 0>
     outlineSettings {
      bUseBrushTransparency=false
      roundingType=HalfHeightRadius
      width=0
      color {colorUseRule=UseColor_Specified, specifiedColor=<rgba 0 0 0 0>}
      cornerRadii=<0 0 0 0>
     }
     tintColor {colorUseRule=UseColor_Specified, specifiedColor=<rgba 1 0.9347 0.088 1>}
     uVRegion {bIsValid=false, max=<0 0>, min=<0 0>}
    }}
   slot {class=HorizontalBoxSlot, properties {horizontalAlignment=HAlign_Center, verticalAlignment=VAlign_Center}}
  }
  Text_PendingTagsWarningLabel {
   widgetClassPath=&TextBlock
   parent=HorizontalBox_0
   widget=Text_PendingTagsWarningLabel
   properties {
    text=Tags awaiting automated nightly script collection.
    font {
     bForceMonospaced=false
     bMaterialIsStencil=false
     fontMaterial=null
     fontObject=&NotoSans
     letterSpacing=0
     monospacedWidth=1
     size=8
     skewAmount=0
     typefaceFontName=NotoSans-Regular
     outlineSettings {
      bApplyOutlineToDropShadows=false
      bMiteredCorners=false
      bSeparateFillAlpha=false
      outlineMaterial=null
      outlineSize=0
      outlineColor=<rgba 0 0 0 1>
     }
    }
   }
   slot {class=HorizontalBoxSlot, properties {padding=<ltrb 4 2 4 2>}}
  }
  Text_PendingTagsWarning {
   widgetClassPath=&TextBlock
   bInherited=true
   parent=VerticalBox_0
   widget=Text_PendingTagsWarning
   properties {
    colorAndOpacity {colorUseRule=UseColor_Specified, specifiedColor=<rgba 1 0.9347 0.088 1>}
    font {
     bForceMonospaced=false
     bMaterialIsStencil=false
     fontMaterial=null
     fontObject=&NotoSans
     letterSpacing=0
     monospacedWidth=1
     size=8
     skewAmount=0
     typefaceFontName=NotoSans-Regular
     outlineSettings {
      bApplyOutlineToDropShadows=false
      bMiteredCorners=false
      bSeparateFillAlpha=false
      outlineMaterial=null
      outlineSize=0
      outlineColor=<rgba 0 0 0 1>
     }
    }
   }
   slot {class=VerticalBoxSlot, properties {padding=<ltrb 2 2 2 2>}}
  }
  Box_FailedTagsWarning {
   widgetClassPath=&HorizontalBox
   bInherited=true
   parent=VerticalBox_25
   widget=Box_FailedTagsWarning
   slot {class=VerticalBoxSlot, properties {padding=<ltrb 0 0 0 2>}}
  }
  Border_1 {
   widgetClassPath=&Border
   parent=Box_FailedTagsWarning
   widget=Border_1
   properties {brushColor=<rgba 0.0469 0.0469 0.0469 1>}
   slot {class=HorizontalBoxSlot, properties {padding=<ltrb 2 0 0 0>, size {sizeRule=Fill, value=1}}}
  }
  VerticalBox {widgetClassPath=&VerticalBox, parent=Border_1, widget=VerticalBox, slot {class=BorderSlot}}
  HorizontalBox {widgetClassPath=&HorizontalBox, parent=VerticalBox, widget=HorizontalBox, slot {class=VerticalBoxSlot}}
  Image_RedWarning {
   widgetClassPath=&Image
   bIsVariable=true
   parent=HorizontalBox
   widget=Image_RedWarning
   properties {brush {
     bIsDynamicallyLoaded=false
     drawAs=Image
     imageType=NoImage
     mirroring=NoMirror
     resourceName=None
     resourceObject=&T-Icon-Warning-64
     tiling=NoTile
     imageSize=<16 16>
     margin=<ltrb 0 0 0 0>
     outlineSettings {
      bUseBrushTransparency=false
      roundingType=HalfHeightRadius
      width=0
      color {colorUseRule=UseColor_Specified, specifiedColor=<rgba 0 0 0 0>}
      cornerRadii=<0 0 0 0>
     }
     tintColor {colorUseRule=UseColor_Specified, specifiedColor=<rgba 1 0.0399 0.0399 1>}
     uVRegion {bIsValid=false, max=<0 0>, min=<0 0>}
    }}
   slot {class=HorizontalBoxSlot, properties {horizontalAlignment=HAlign_Center, verticalAlignment=VAlign_Center}}
  }
  Text_FailedTagsWarningLabel {
   widgetClassPath=&TextBlock
   parent=HorizontalBox
   widget=Text_FailedTagsWarningLabel
   properties {
    text=Automated nightly script failed to find tags. Check map for missing tags.
    font {
     bForceMonospaced=false
     bMaterialIsStencil=false
     fontMaterial=null
     fontObject=&NotoSans
     letterSpacing=0
     monospacedWidth=1
     size=8
     skewAmount=0
     typefaceFontName=NotoSans-Regular
     outlineSettings {
      bApplyOutlineToDropShadows=false
      bMiteredCorners=false
      bSeparateFillAlpha=false
      outlineMaterial=null
      outlineSize=0
      outlineColor=<rgba 0 0 0 1>
     }
    }
   }
   slot {class=HorizontalBoxSlot, properties {padding=<ltrb 4 2 4 2>}}
  }
  Text_FailedTagsWarning {
   widgetClassPath=&TextBlock
   bInherited=true
   parent=VerticalBox
   widget=Text_FailedTagsWarning
   properties {
    colorAndOpacity {colorUseRule=UseColor_Specified, specifiedColor=<rgba 1 0.0191 0.0191 1>}
    font {
     bForceMonospaced=false
     bMaterialIsStencil=false
     fontMaterial=null
     fontObject=&NotoSans
     letterSpacing=0
     monospacedWidth=1
     size=8
     skewAmount=0
     typefaceFontName=NotoSans-Regular
     outlineSettings {
      bApplyOutlineToDropShadows=false
      bMiteredCorners=false
      bSeparateFillAlpha=false
      outlineMaterial=null
      outlineSize=0
      outlineColor=<rgba 0 0 0 1>
     }
    }
   }
   slot {class=VerticalBoxSlot, properties {padding=<ltrb 2 2 2 2>}}
  }
  ExpandableArea_AdditionalInfo {
   widgetClassPath=/Script/Blutility.EditorUtilityExpandableArea
   bInherited=true
   bIsVariable=true
   parent=VerticalBox_25
   widget=ExpandableArea_AdditionalInfo
   properties {areaPadding=<ltrb 4 2 4 2>}
   slot {class=VerticalBoxSlot, properties {padding=<ltrb 6 6 6 6>}}
  }
  Text_AdditionalInfoHeader {
   widgetClassPath=&TextBlock
   namedSlotHost=ExpandableArea_AdditionalInfo
   widget=Text_AdditionalInfoHeader
   properties {
    text=Additional Info
    colorAndOpacity {colorUseRule=UseColor_Specified, specifiedColor=<rgba 0.9941 0.9747 1 1>}
    font {
     bForceMonospaced=false
     bMaterialIsStencil=false
     fontMaterial=null
     fontObject=&NotoSans
     letterSpacing=0
     monospacedWidth=1
     size=8
     skewAmount=0
     typefaceFontName=NotoSans-Regular
     outlineSettings {
      bApplyOutlineToDropShadows=false
      bMiteredCorners=false
      bSeparateFillAlpha=false
      outlineMaterial=null
      outlineSize=0
      outlineColor=<rgba 0 0 0 1>
     }
    }
   }
  }
  Text_AdditionalInfo {
   widgetClassPath=&TextBlock
   bInherited=true
   namedSlotHost=ExpandableArea_AdditionalInfo
   widget=Text_AdditionalInfo
   properties {font {
     bForceMonospaced=false
     bMaterialIsStencil=false
     fontMaterial=null
     fontObject=&NotoSans
     letterSpacing=0
     monospacedWidth=1
     size=8
     skewAmount=0
     typefaceFontName=NotoSans-Regular
     outlineSettings {
      bApplyOutlineToDropShadows=false
      bMiteredCorners=false
      bSeparateFillAlpha=false
      outlineMaterial=null
      outlineSize=0
      outlineColor=<rgba 0 0 0 1>
     }
    }}
  }
 ]
}
