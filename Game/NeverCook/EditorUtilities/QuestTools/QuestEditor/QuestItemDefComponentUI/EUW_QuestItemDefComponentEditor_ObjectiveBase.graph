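# Editor utility widget for editing a quest item definition objective component in the quest editor, built on FortQuestItemDefComponentEditor_Objective.
# It lays out inherited widgets: a header with name and remove button, an add-extension-data button (default text references Gate objectives), and a property details view excluding ObjectivesExtensionData.
# Two functions generate combo box content and item widgets, each initialized with a map name.
# The event graph holds only unconnected event stubs.
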EditorUtilityWidgetBlueprint /Game/NeverCook/EditorUtilities/QuestTools/QuestEditor/QuestItemDefComponentUI/EUW_QuestItemDefComponentEditor_ObjectiveBase
graphs [
 EventGraph {nodes [
   Event PreConstruct()
   Event Construct()
   Event Tick()
  ]}
 OnComboBoxGenerateContentWidget {nodes [
   n1 = OnComboBoxGenerateContentWidget!() -> (then, Item:Name) {is=title K2Node_FunctionEntry}
   > n2 = Create Widget!(
    execute:Exec
    WorldContextObject:Object Reference
    Class:User Widget Class Reference=/Game/NeverCook/EditorUtilities/QuestTools/QuestEditor/QuestItemDefComponentUI/EditorUtilityWidgets/ComboBoxMapNameContent.ComboBoxMapNameContent_C
    OwningPlayer:Player Controller Object Reference
   ) -> (then:Exec, ReturnValue:Combo Box Map Name Content Object Reference)
   > InitMapEntry!(execute, self:Combo Box Map Name Content Object Reference=@n2, InMapName:Name=@n1) -> (then)
   > Return Node!(execute, ReturnValue:Widget Object Reference=@n2)
  ]}
 OnComboBoxGenerateItemWidget {nodes [
   n1 = OnComboBoxGenerateItemWidget!() -> (then, Item:Name) {is=title K2Node_FunctionEntry}
   > n2 = Create Widget(Class=/Game/NeverCook/EditorUtilities/QuestTools/QuestEditor/QuestItemDefComponentUI/EditorUtilityWidgets/ComboBoxMapNameItem.ComboBoxMapNameItem_C) -> (ReturnValue:Combo Box Map Name Item Object Reference)
   > InitMapEntry(self:Combo Box Map Name Item Object Reference=@n2, InMapName=@n1)
   > Return Node(ReturnValue=@n2)
  ]}
]
variables []
widgets {
 info {
  inheritedWidgetCount=6
  parentClass=/Script/FortniteEditor.FortQuestItemDefComponentEditor_Objective
  rootWidgetClass=&HorizontalBox
  widgetCount=14
 }
 widgets [
  HorizontalBox_1765 {widgetClassPath=&HorizontalBox, widget=HorizontalBox_1765}
  VerticalBox {
   widgetClassPath=&VerticalBox
   parent=HorizontalBox_1765
   widget=VerticalBox
   slot {class=HorizontalBoxSlot, properties {horizontalAlignment=HAlign_Left, size {sizeRule=Fill, value=1}}}
  }
  Header_Border {
   widgetClassPath=&Border
   parent=VerticalBox
   widget=Header_Border
   properties {brushColor=<rgba 0.0284 0.0284 0.0284 1>, padding=<ltrb 4 6 4 6>}
   slot {class=VerticalBoxSlot, properties {verticalAlignment=VAlign_Center}}
  }
  HorizontalBox_0 {
   widgetClassPath=&HorizontalBox
   parent=Header_Border
   widget=HorizontalBox_0
   slot {class=BorderSlot, properties {padding=<ltrb 4 6 4 6>}}
  }
  Text_ComponentName {
   widgetClassPath=&CommonTextBlock
   bInherited=true
   parent=HorizontalBox_0
   widget=Text_ComponentName
   properties {
    text=Objective Gate
    colorAndOpacity {colorUseRule=UseColor_Specified, specifiedColor=<rgba 0.9671 1 0.9899 1>}
    font {
     bForceMonospaced=false
     bMaterialIsStencil=false
     fontMaterial=null
     fontObject=&NotoSans
     letterSpacing=0
     monospacedWidth=1
     size=8
     skewAmount=0
     typefaceFontName=NotoSans-Regular
     outlineSettings {
      bApplyOutlineToDropShadows=false
      bMiteredCorners=false
      bSeparateFillAlpha=false
      outlineMaterial=null
      outlineSize=0
      outlineColor=<rgba 0 0 0 1>
     }
    }
    shadowOffset=<0 0>
   }
   slot {class=HorizontalBoxSlot}
  }
  Button_RemoveObjectiveExtensionData {
   widgetClassPath=&Button
   bInherited=true
   parent=HorizontalBox_0
   widget=Button_RemoveObjectiveExtensionData
   properties {widgetStyle {
     clickedSlateSound {resourceObject=null}
     disabled {
      bIsDynamicallyLoaded=false
      drawAs=RoundedBox
      imageType=NoImage
      mirroring=NoMirror
      resourceName=None
      resourceObject=null
      tiling=NoTile
      imageSize=<0 0>
      margin=<ltrb 0 0 0 0>
      outlineSettings {
       bUseBrushTransparency=false
       roundingType=FixedRadius
       width=1
       color {colorUseRule=UseColor_ColorTable, specifiedColor=<rgba 0 0 0 0>}
       cornerRadii=<4 4 4 4>
      }
      tintColor {colorUseRule=UseColor_Specified, specifiedColor=<rgba 0.0395 0.0395 0.0395 1>}
      uVRegion {bIsValid=false, max=<0 0>, min=<0 0>}
     }
     disabledForeground {colorUseRule=UseColor_ColorTable, specifiedColor=<rgba 0 0 0 0>}
     hovered {
      bIsDynamicallyLoaded=false
      drawAs=Image
      imageType=NoImage
      mirroring=NoMirror
      resourceName=None
      resourceObject=/Game/UI/Foundation/Textures/Buttons/T-SlotIcon-Remove-32
      tiling=NoTile
      imageSize=<16 16>
      margin=<ltrb 0 0 0 0>
      outlineSettings {
       bUseBrushTransparency=false
       roundingType=FixedRadius
       width=1
       color {colorUseRule=UseColor_ColorTable, specifiedColor=<rgba 0 0 0 0>}
       cornerRadii=<4 4 4 4>
      }
      tintColor {colorUseRule=UseColor_Specified, specifiedColor=<rgba 1 0 0 1>}
      uVRegion {bIsValid=false, max=<0 0>, min=<0 0>}
     }
     hoveredForeground {colorUseRule=UseColor_ColorTable, specifiedColor=<rgba 0 0 0 0>}
     hoveredSlateSound {resourceObject=null}
     normal {
      bIsDynamicallyLoaded=false
      drawAs=Image
      imageType=NoImage
      mirroring=NoMirror
      resourceName=None
      resourceObject=&T-SlotIcon-Remove-32
      tiling=NoTile
      imageSize=<16 16>
      margin=<ltrb 0 0 0 0>
      outlineSettings {
       bUseBrushTransparency=false
       roundingType=FixedRadius
       width=1
       color {colorUseRule=UseColor_ColorTable, specifiedColor=<rgba 0 0 0 0>}
       cornerRadii=<4 4 4 4>
      }
      tintColor {colorUseRule=UseColor_Specified, specifiedColor=<rgba 0.9671 1 0.9899 1>}
      uVRegion {bIsValid=false, max=<0 0>, min=<0 0>}
     }
     normalForeground {colorUseRule=UseColor_ColorTable, specifiedColor=<rgba 0 0 0 0>}
     normalPadding=<ltrb 12 1.5 12 1.5>
     pressed {
      bIsDynamicallyLoaded=false
      drawAs=Image
      imageType=NoImage
      mirroring=NoMirror
      resourceName=None
      resourceObject=&T-SlotIcon-Remove-32
      tiling=NoTile
      imageSize=<16 16>
      margin=<ltrb 0 0 0 0>
      outlineSettings {
       bUseBrushTransparency=false
       roundingType=FixedRadius
       width=1
       color {colorUseRule=UseColor_ColorTable, specifiedColor=<rgba 0 0 0 0>}
       cornerRadii=<4 4 4 4>
      }
      tintColor {colorUseRule=UseColor_Specified, specifiedColor=<rgba 1 0 0 1>}
      uVRegion {bIsValid=false, max=<0 0>, min=<0 0>}
     }
     pressedForeground {colorUseRule=UseColor_ColorTable, specifiedColor=<rgba 0 0 0 0>}
     pressedPadding=<ltrb 12 2.5 12 0.5>
     pressedSlateSound {resourceObject=null}
    }}
   slot {class=HorizontalBoxSlot, properties {horizontalAlignment=HAlign_Right, verticalAlignment=VAlign_Center, size {sizeRule=Fill, value=1}}}
  }
  Spacer_58 {widgetClassPath=&Spacer, parent=VerticalBox, widget=Spacer_58, slot {class=VerticalBoxSlot}}
  Border_60 {
   widgetClassPath=&Border
   parent=VerticalBox
   widget=Border_60
   properties {brushColor=<rgba 0.0176 0.0176 0.0176 1>, padding=<ltrb 0 0 0 0>}
   slot {class=VerticalBoxSlot}
  }
  VerticalBox_0 {
   widgetClassPath=&VerticalBox
   parent=Border_60
   widget=VerticalBox_0
   slot {class=BorderSlot, properties {padding=<ltrb 0 0 0 0>}}
  }
  Button_AddObjectiveExtensionData {
   widgetClassPath=&Button
   bInherited=true
   parent=VerticalBox_0
   widget=Button_AddObjectiveExtensionData
   properties {widgetStyle {
     clickedSlateSound {resourceObject=null}
     disabled {
      bIsDynamicallyLoaded=false
      drawAs=RoundedBox
      imageType=NoImage
      mirroring=NoMirror
      resourceName=None
      resourceObject=null
      tiling=NoTile
      imageSize=<0 0>
      margin=<ltrb 0 0 0 0>
      outlineSettings {
       bUseBrushTransparency=false
       roundingType=FixedRadius
       width=1
       color {colorUseRule=UseColor_ColorTable, specifiedColor=<rgba 0 0 0 0>}
       cornerRadii=<4 4 4 4>
      }
      tintColor {colorUseRule=UseColor_Specified, specifiedColor=<rgba 0.0395 0.0395 0.0395 1>}
      uVRegion {bIsValid=false, max=<0 0>, min=<0 0>}
     }
     disabledForeground {colorUseRule=UseColor_ColorTable, specifiedColor=<rgba 0 0 0 0>}
     hovered {
      bIsDynamicallyLoaded=false
      drawAs=RoundedBox
      imageType=NoImage
      mirroring=NoMirror
      resourceName=None
      resourceObject=null
      tiling=NoTile
      imageSize=<32 32>
      margin=<ltrb 0 0 0 0>
      outlineSettings {
       bUseBrushTransparency=false
       roundingType=FixedRadius
       width=1
       color {colorUseRule=UseColor_ColorTable, specifiedColor=<rgba 0 0 0 0>}
       cornerRadii=<4 4 4 4>
      }
      tintColor {colorUseRule=UseColor_Specified, specifiedColor=<rgba 0.0953 0.0953 0.0953 1>}
      uVRegion {bIsValid=false, max=<0 0>, min=<0 0>}
     }
     hoveredForeground {colorUseRule=UseColor_ColorTable, specifiedColor=<rgba 0 0 0 0>}
     hoveredSlateSound {resourceObject=null}
     normal {
      bIsDynamicallyLoaded=false
      drawAs=RoundedBox
      imageType=NoImage
      mirroring=NoMirror
      resourceName=None
      resourceObject=null
      tiling=NoTile
      imageSize=<32 32>
      margin=<ltrb 0 0 0 0>
      outlineSettings {
       bUseBrushTransparency=false
       roundingType=FixedRadius
       width=1
       color {colorUseRule=UseColor_ColorTable, specifiedColor=<rgba 0 0 0 0>}
       cornerRadii=<4 4 4 4>
      }
      tintColor {colorUseRule=UseColor_Specified, specifiedColor=<rgba 0.0395 0.0395 0.0395 1>}
      uVRegion {bIsValid=false, max=<0 0>, min=<0 0>}
     }
     normalForeground {colorUseRule=UseColor_ColorTable, specifiedColor=<rgba 0 0 0 0>}
     normalPadding=<ltrb 12 1.5 12 1.5>
     pressed {
      bIsDynamicallyLoaded=false
      drawAs=RoundedBox
      imageType=NoImage
      mirroring=NoMirror
      resourceName=None
      resourceObject=null
      tiling=NoTile
      imageSize=<32 32>
      margin=<ltrb 0 0 0 0>
      outlineSettings {
       bUseBrushTransparency=false
       roundingType=FixedRadius
       width=1
       color {colorUseRule=UseColor_ColorTable, specifiedColor=<rgba 0 0 0 0>}
       cornerRadii=<4 4 4 4>
      }
      tintColor {colorUseRule=UseColor_Specified, specifiedColor=<rgba 0.0284 0.0284 0.0284 1>}
      uVRegion {bIsValid=false, max=<0 0>, min=<0 0>}
     }
     pressedForeground {colorUseRule=UseColor_ColorTable, specifiedColor=<rgba 0 0 0 0>}
     pressedPadding=<ltrb 12 2.5 12 0.5>
     pressedSlateSound {resourceObject=null}
    }}
   slot {class=VerticalBoxSlot, properties {verticalAlignment=VAlign_Center, padding=<ltrb 4 2 4 2>}}
  }
  Text_AddExtensionData {
   widgetClassPath=&CommonTextBlock
   bInherited=true
   parent=Button_AddObjectiveExtensionData
   widget=Text_AddExtensionData
   properties {
    autoWrapText=true
    bIsScrollingEnabled=false
    justification=Center
    text=Add Gate Objective Data
    font {
     bForceMonospaced=false
     bMaterialIsStencil=false
     fontMaterial=null
     fontObject=&NotoSans
     letterSpacing=0
     monospacedWidth=1
     size=10
     skewAmount=0
     typefaceFontName=NotoSans-Regular
     outlineSettings {
      bApplyOutlineToDropShadows=false
      bMiteredCorners=false
      bSeparateFillAlpha=false
      outlineMaterial=null
      outlineSize=0
      outlineColor=<rgba 0 0 0 1>
     }
    }
    shadowOffset=<0 0>
   }
   slot {class=ButtonSlot, properties {horizontalAlignment=HAlign_Fill}}
  }
  SizeBox_0 {
   widgetClassPath=&SizeBox
   parent=VerticalBox_0
   widget=SizeBox_0
   properties {bOverride_MinDesiredWidth=true, minDesiredWidth=600}
   slot {class=VerticalBoxSlot}
  }
  DetailView_ComponentProperties {
   widgetClassPath=/Script/PropertiesDetailView
   bInherited=true
   bIsVariable=true
   parent=SizeBox_0
   widget=DetailView_ComponentProperties
   properties {
    bShowAllProperties=true
    bShowPropertiesThatMatchGeneratedPropertyPaths=true
    visibility=SelfHitTestInvisible
    propertiesToExclude [ObjectivesExtensionData]
    propertyColumns [{fillCoefficient=1, propertiesDisplayData []}]
   }
   slot {class=SizeBoxSlot}
  }
  Button_Copy {widgetClassPath=&CommonButtonBase, bInherited=true, widget=null}
 ]
}
