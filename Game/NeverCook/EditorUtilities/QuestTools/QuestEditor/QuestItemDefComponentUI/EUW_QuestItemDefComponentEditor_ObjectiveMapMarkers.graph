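# Editor utility widget in the quest editor for editing an objective's map marker component data.
# It is layout only: inherited widgets bind to the C++ parent FortQuestItemDefComponentEditor_ObjectiveMapMarkers, and the event graph holds only unwired events.
# A header with a remove button sits above an add-data button, marker tag status, icon-override error row and icon override combo box.
# A details view shows location tag properties and excludes ObjectiveIconOverride, which has its own combo box row.
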EditorUtilityWidgetBlueprint /Game/NeverCook/EditorUtilities/QuestTools/QuestEditor/QuestItemDefComponentUI/EUW_QuestItemDefComponentEditor_ObjectiveMapMarkers
graphs [EventGraph {nodes [
   Event PreConstruct()
   Event Construct()
   Event Tick()
  ]}]
variables []
widgets {
 info {
  inheritedWidgetCount=11
  parentClass=/Script/FortniteEditor.FortQuestItemDefComponentEditor_ObjectiveMapMarkers
  rootWidgetClass=&HorizontalBox
  widgetCount=24
 }
 widgets [
  HorizontalBox_1765 {widgetClassPath=&HorizontalBox, widget=HorizontalBox_1765}
  VerticalBox {
   widgetClassPath=&VerticalBox
   parent=HorizontalBox_1765
   widget=VerticalBox
   slot {class=HorizontalBoxSlot, properties {horizontalAlignment=HAlign_Left, size {sizeRule=Fill, value=1}}}
  }
  Border_0 {
   widgetClassPath=&Border
   parent=VerticalBox
   widget=Border_0
   properties {brushColor=<rgba 0.0284 0.0284 0.0284 1>, padding=<ltrb 4 6 4 6>}
   slot {class=VerticalBoxSlot, properties {verticalAlignment=VAlign_Center}}
  }
  HorizontalBox_0 {
   widgetClassPath=&HorizontalBox
   parent=Border_0
   widget=HorizontalBox_0
   slot {class=BorderSlot, properties {padding=<ltrb 4 6 4 6>}}
  }
  Text_ComponentName {
   widgetClassPath=&CommonTextBlock
   bInherited=true
   parent=HorizontalBox_0
   widget=Text_ComponentName
   properties {
    text=Objective Map Markers
    colorAndOpacity {colorUseRule=UseColor_Specified, specifiedColor=<rgba 0.9671 1 0.9899 1>}
    font {
     bForceMonospaced=false
     bMaterialIsStencil=false
     fontMaterial=null
     fontObject=&NotoSans
     letterSpacing=0
     monospacedWidth=1
     size=8
     skewAmount=0
     typefaceFontName=NotoSans-Regular
     outlineSettings {
      bApplyOutlineToDropShadows=false
      bMiteredCorners=false
      bSeparateFillAlpha=false
      outlineMaterial=null
      outlineSize=0
      outlineColor=<rgba 0 0 0 1>
     }
    }
    shadowOffset=<0 0>
   }
   slot {class=HorizontalBoxSlot}
  }
  Button_RemoveObjectiveExtensionData {
   widgetClassPath=&Button
   bInherited=true
   parent=HorizontalBox_0
   widget=Button_RemoveObjectiveExtensionData
   properties {widgetStyle {
     clickedSlateSound {resourceObject=null}
     disabled {
      bIsDynamicallyLoaded=false
      drawAs=RoundedBox
      imageType=NoImage
      mirroring=NoMirror
      resourceName=None
      resourceObject=null
      tiling=NoTile
      imageSize=<0 0>
      margin=<ltrb 0 0 0 0>
      outlineSettings {
       bUseBrushTransparency=false
       roundingType=FixedRadius
       width=1
       color {colorUseRule=UseColor_ColorTable, specifiedColor=<rgba 0 0 0 0>}
       cornerRadii=<4 4 4 4>
      }
      tintColor {colorUseRule=UseColor_Specified, specifiedColor=<rgba 0.0395 0.0395 0.0395 1>}
      uVRegion {bIsValid=false, max=<0 0>, min=<0 0>}
     }
     disabledForeground {colorUseRule=UseColor_ColorTable, specifiedColor=<rgba 0 0 0 0>}
     hovered {
      bIsDynamicallyLoaded=false
      drawAs=Image
      imageType=NoImage
      mirroring=NoMirror
      resourceName=None
      resourceObject=/Game/UI/Foundation/Textures/Buttons/T-SlotIcon-Remove-32
      tiling=NoTile
      imageSize=<16 16>
      margin=<ltrb 0 0 0 0>
      outlineSettings {
       bUseBrushTransparency=false
       roundingType=FixedRadius
       width=1
       color {colorUseRule=UseColor_ColorTable, specifiedColor=<rgba 0 0 0 0>}
       cornerRadii=<4 4 4 4>
      }
      tintColor {colorUseRule=UseColor_Specified, specifiedColor=<rgba 1 0 0 1>}
      uVRegion {bIsValid=false, max=<0 0>, min=<0 0>}
     }
     hoveredForeground {colorUseRule=UseColor_ColorTable, specifiedColor=<rgba 0 0 0 0>}
     hoveredSlateSound {resourceObject=null}
     normal {
      bIsDynamicallyLoaded=false
      drawAs=Image
      imageType=NoImage
      mirroring=NoMirror
      resourceName=None
      resourceObject=&T-SlotIcon-Remove-32
      tiling=NoTile
      imageSize=<16 16>
      margin=<ltrb 0 0 0 0>
      outlineSettings {
       bUseBrushTransparency=false
       roundingType=FixedRadius
       width=1
       color {colorUseRule=UseColor_ColorTable, specifiedColor=<rgba 0 0 0 0>}
       cornerRadii=<4 4 4 4>
      }
      tintColor {colorUseRule=UseColor_Specified, specifiedColor=<rgba 0.9671 1 0.9899 1>}
      uVRegion {bIsValid=false, max=<0 0>, min=<0 0>}
     }
     normalForeground {colorUseRule=UseColor_ColorTable, specifiedColor=<rgba 0 0 0 0>}
     normalPadding=<ltrb 12 1.5 12 1.5>
     pressed {
      bIsDynamicallyLoaded=false
      drawAs=Image
      imageType=NoImage
      mirroring=NoMirror
      resourceName=None
      resourceObject=&T-SlotIcon-Remove-32
      tiling=NoTile
      imageSize=<16 16>
      margin=<ltrb 0 0 0 0>
      outlineSettings {
       bUseBrushTransparency=false
       roundingType=FixedRadius
       width=1
       color {colorUseRule=UseColor_ColorTable, specifiedColor=<rgba 0 0 0 0>}
       cornerRadii=<4 4 4 4>
      }
      tintColor {colorUseRule=UseColor_Specified, specifiedColor=<rgba 1 0 0 1>}
      uVRegion {bIsValid=false, max=<0 0>, min=<0 0>}
     }
     pressedForeground {colorUseRule=UseColor_ColorTable, specifiedColor=<rgba 0 0 0 0>}
     pressedPadding=<ltrb 12 2.5 12 0.5>
     pressedSlateSound {resourceObject=null}
    }}
   slot {class=HorizontalBoxSlot, properties {horizontalAlignment=HAlign_Right, verticalAlignment=VAlign_Center, size {sizeRule=Fill, value=1}}}
  }
  Spacer_58 {widgetClassPath=&Spacer, parent=VerticalBox, widget=Spacer_58, slot {class=VerticalBoxSlot}}
  Border_34 {
   widgetClassPath=&Border
   parent=VerticalBox
   widget=Border_34
   properties {brushColor=<rgba 0.0176 0.0176 0.0176 1>, padding=<ltrb 0 0 0 0>}
   slot {class=VerticalBoxSlot}
  }
  VerticalBox_86 {
   widgetClassPath=&VerticalBox
   parent=Border_34
   widget=VerticalBox_86
   slot {class=BorderSlot, properties {padding=<ltrb 0 0 0 0>}}
  }
  Button_AddObjectiveExtensionData {
   widgetClassPath=&Button
   bInherited=true
   parent=VerticalBox_86
   widget=Button_AddObjectiveExtensionData
   properties {widgetStyle {
     clickedSlateSound {resourceObject=null}
     disabled {
      bIsDynamicallyLoaded=false
      drawAs=RoundedBox
      imageType=NoImage
      mirroring=NoMirror
      resourceName=None
      resourceObject=null
      tiling=NoTile
      imageSize=<0 0>
      margin=<ltrb 0 0 0 0>
      outlineSettings {
       bUseBrushTransparency=false
       roundingType=FixedRadius
       width=1
       color {colorUseRule=UseColor_ColorTable, specifiedColor=<rgba 0 0 0 0>}
       cornerRadii=<4 4 4 4>
      }
      tintColor {colorUseRule=UseColor_Specified, specifiedColor=<rgba 0.0395 0.0395 0.0395 1>}
      uVRegion {bIsValid=false, max=<0 0>, min=<0 0>}
     }
     disabledForeground {colorUseRule=UseColor_ColorTable, specifiedColor=<rgba 0 0 0 0>}
     hovered {
      bIsDynamicallyLoaded=false
      drawAs=RoundedBox
      imageType=NoImage
      mirroring=NoMirror
      resourceName=None
      resourceObject=null
      tiling=NoTile
      imageSize=<32 32>
      margin=<ltrb 0 0 0 0>
      outlineSettings {
       bUseBrushTransparency=false
       roundingType=FixedRadius
       width=1
       color {colorUseRule=UseColor_ColorTable, specifiedColor=<rgba 0 0 0 0>}
       cornerRadii=<4 4 4 4>
      }
      tintColor {colorUseRule=UseColor_Specified, specifiedColor=<rgba 0.0953 0.0953 0.0953 1>}
      uVRegion {bIsValid=false, max=<0 0>, min=<0 0>}
     }
     hoveredForeground {colorUseRule=UseColor_ColorTable, specifiedColor=<rgba 0 0 0 0>}
     hoveredSlateSound {resourceObject=null}
     normal {
      bIsDynamicallyLoaded=false
      drawAs=RoundedBox
      imageType=NoImage
      mirroring=NoMirror
      resourceName=None
      resourceObject=null
      tiling=NoTile
      imageSize=<32 32>
      margin=<ltrb 0 0 0 0>
      outlineSettings {
       bUseBrushTransparency=false
       roundingType=FixedRadius
       width=1
       color {colorUseRule=UseColor_ColorTable, specifiedColor=<rgba 0 0 0 0>}
       cornerRadii=<4 4 4 4>
      }
      tintColor {colorUseRule=UseColor_Specified, specifiedColor=<rgba 0.0395 0.0395 0.0395 1>}
      uVRegion {bIsValid=false, max=<0 0>, min=<0 0>}
     }
     normalForeground {colorUseRule=UseColor_ColorTable, specifiedColor=<rgba 0 0 0 0>}
     normalPadding=<ltrb 12 1.5 12 1.5>
     pressed {
      bIsDynamicallyLoaded=false
      drawAs=RoundedBox
      imageType=NoImage
      mirroring=NoMirror
      resourceName=None
      resourceObject=null
      tiling=NoTile
      imageSize=<32 32>
      margin=<ltrb 0 0 0 0>
      outlineSettings {
       bUseBrushTransparency=false
       roundingType=FixedRadius
       width=1
       color {colorUseRule=UseColor_ColorTable, specifiedColor=<rgba 0 0 0 0>}
       cornerRadii=<4 4 4 4>
      }
      tintColor {colorUseRule=UseColor_Specified, specifiedColor=<rgba 0.0284 0.0284 0.0284 1>}
      uVRegion {bIsValid=false, max=<0 0>, min=<0 0>}
     }
     pressedForeground {colorUseRule=UseColor_ColorTable, specifiedColor=<rgba 0 0 0 0>}
     pressedPadding=<ltrb 12 2.5 12 0.5>
     pressedSlateSound {resourceObject=null}
    }}
   slot {class=VerticalBoxSlot, properties {verticalAlignment=VAlign_Center, padding=<ltrb 4 2 4 2>}}
  }
  Text_AddExtensionData {
   widgetClassPath=&CommonTextBlock
   bInherited=true
   parent=Button_AddObjectiveExtensionData
   widget=Text_AddExtensionData
   properties {
    autoWrapText=true
    bIsScrollingEnabled=false
    justification=Center
    text=Add Map Marker Objective Data
    font {
     bForceMonospaced=false
     bMaterialIsStencil=false
     fontMaterial=null
     fontObject=&NotoSans
     letterSpacing=0
     monospacedWidth=1
     size=10
     skewAmount=0
     typefaceFontName=NotoSans-Regular
     outlineSettings {
      bApplyOutlineToDropShadows=false
      bMiteredCorners=false
      bSeparateFillAlpha=false
      outlineMaterial=null
      outlineSize=0
      outlineColor=<rgba 0 0 0 1>
     }
    }
    shadowOffset=<0 0>
   }
   slot {class=ButtonSlot, properties {horizontalAlignment=HAlign_Fill}}
  }
  Widget_MapMarkerTagsStatus {
   widgetClassPath=&/MapMarkerStatus
   bInherited=true
   bIsVariable=true
   parent=VerticalBox_86
   widget=Widget_MapMarkerTagsStatus
   slot {class=VerticalBoxSlot}
  }
  Border_TagIconOverrideError {
   widgetClassPath=&Border
   bInherited=true
   parent=VerticalBox_86
   widget=Border_TagIconOverrideError
   properties {brushColor=<rgba 0.0168 0.0168 0.0168 1>}
   slot {class=VerticalBoxSlot}
  }
  HorizontalBox_2 {widgetClassPath=&HorizontalBox, parent=Border_TagIconOverrideError, widget=HorizontalBox_2, slot {class=BorderSlot}}
  Image_RedWarning2 {
   widgetClassPath=&Image
   bIsVariable=true
   parent=HorizontalBox_2
   widget=Image_RedWarning2
   properties {brush {
     bIsDynamicallyLoaded=false
     drawAs=Image
     imageType=NoImage
     mirroring=NoMirror
     resourceName=None
     resourceObject=/Game/UI/Foundation/Shell/Textures/T-Icon-Warning-64
     tiling=NoTile
     imageSize=<16 16>
     margin=<ltrb 0 0 0 0>
     outlineSettings {
      bUseBrushTransparency=false
      roundingType=HalfHeightRadius
      width=0
      color {colorUseRule=UseColor_Specified, specifiedColor=<rgba 0 0 0 0>}
      cornerRadii=<0 0 0 0>
     }
     tintColor {colorUseRule=UseColor_Specified, specifiedColor=<rgba 1 0.0399 0.0399 1>}
     uVRegion {bIsValid=false, max=<0 0>, min=<0 0>}
    }}
   slot {class=HorizontalBoxSlot, properties {horizontalAlignment=HAlign_Center, verticalAlignment=VAlign_Center}}
  }
  Text_TagIconOverrideError {
   widgetClassPath=&TextBlock
   bInherited=true
   parent=HorizontalBox_2
   widget=Text_TagIconOverrideError
   properties {
    autoWrapText=true
    wrapTextAt=500
    wrappingPolicy=AllowPerCharacterWrapping
    font {
     bForceMonospaced=false
     bMaterialIsStencil=false
     fontMaterial=null
     fontObject=&NotoSans
     letterSpacing=0
     monospacedWidth=1
     size=8
     skewAmount=0
     typefaceFontName=NotoSans-Regular
     outlineSettings {
      bApplyOutlineToDropShadows=false
      bMiteredCorners=false
      bSeparateFillAlpha=false
      outlineMaterial=null
      outlineSize=0
      outlineColor=<rgba 0 0 0 1>
     }
    }
   }
   slot {class=HorizontalBoxSlot, properties {padding=<ltrb 4 2 4 2>}}
  }
  Border {
   widgetClassPath=&Border
   parent=VerticalBox_86
   widget=Border
   properties {brushColor=<rgba 0.0168 0.0168 0.0168 1>, padding=<ltrb 0 0 0 0>}
   slot {class=VerticalBoxSlot, properties {padding=<ltrb 0 2 0 2>}}
  }
  HorizontalBox_IconOverride {
   widgetClassPath=&HorizontalBox
   bInherited=true
   parent=Border
   widget=HorizontalBox_IconOverride
   slot {class=BorderSlot, properties {padding=<ltrb 0 0 0 0>}}
  }
  SizeBox {
   widgetClassPath=&SizeBox
   parent=HorizontalBox_IconOverride
   widget=SizeBox
   properties {bOverride_MinDesiredWidth=true, minDesiredWidth=120}
   slot {class=HorizontalBoxSlot, properties {padding=<ltrb 0 5 0 5>}}
  }
  CommonTextBlock {
   widgetClassPath=&CommonTextBlock
   parent=SizeBox
   widget=CommonTextBlock
   properties {
    text=Icon Override
    toolTipText=Override default map marker icon displayed for this objective.
    visibility=Visible
    font {
     bForceMonospaced=false
     bMaterialIsStencil=false
     fontMaterial=null
     fontObject=&NotoSans
     letterSpacing=0
     monospacedWidth=1
     size=8
     skewAmount=0
     typefaceFontName=NotoSans-Regular
     outlineSettings {
      bApplyOutlineToDropShadows=false
      bMiteredCorners=false
      bSeparateFillAlpha=false
      outlineMaterial=null
      outlineSize=0
      outlineColor=<rgba 0 0 0 1>
     }
    }
    shadowOffset=<0 0>
   }
   slot {class=SizeBoxSlot, properties {horizontalAlignment=HAlign_Left, verticalAlignment=VAlign_Center, padding=<ltrb 10 0 10 0>}}
  }
  ComboBox_IconOverrides {
   widgetClassPath=/Game/NeverCook/EditorUtilities/QuestTools/QuestEditor/QuestItemDefComponentUI/EditorUtilityWidgets/MapIconOverrideComboBox
   bInherited=true
   bIsVariable=true
   parent=HorizontalBox_IconOverride
   widget=ComboBox_IconOverrides
   slot {class=HorizontalBoxSlot, properties {padding=<ltrb 0 4 2 4>, size {sizeRule=Fill, value=1}}}
  }
  SizeBox_0 {
   widgetClassPath=&SizeBox
   parent=VerticalBox_86
   widget=SizeBox_0
   properties {bOverride_MinDesiredWidth=true, minDesiredWidth=600}
   slot {class=VerticalBoxSlot}
  }
  DetailView_ComponentProperties {
   widgetClassPath=/Script/PropertiesDetailView
   bInherited=true
   bIsVariable=true
   parent=SizeBox_0
   widget=DetailView_ComponentProperties
   properties {
    visibility=SelfHitTestInvisible
    propertiesToExclude [ObjectiveIconOverride]
    propertyColumns [{
      fillCoefficient=1
      propertiesDisplayData [
       {
        bForceShowParentProperty=false
        bForceShowProperty=false
        bShowChildren=true
        bShowContainerProperty=true
        bShowName=true
        bShowPropertyButtons=true
        bShowValue=true
        containerObjectName=None
        propertyName=QuestObjectiveLocationTags
        propertyNameOverride=""
       }
       {
        bForceShowParentProperty=false
        bForceShowProperty=false
        bShowChildren=true
        bShowContainerProperty=true
        bShowName=true
        bShowPropertyButtons=true
        bShowValue=true
        containerObjectName=None
        propertyName=bHideQuestObjectiveLocationTagsIfEventTagsAreActive
        propertyNameOverride=""
       }
       {
        bForceShowParentProperty=false
        bForceShowProperty=false
        bShowChildren=true
        bShowContainerProperty=true
        bShowName=true
        bShowPropertyButtons=true
        bShowValue=true
        containerObjectName=None
        propertyName=EventObjectiveLocationTags
        propertyNameOverride=""
       }
      ]
     }]
   }
   slot {class=SizeBoxSlot}
  }
  Button_Copy {widgetClassPath=&CommonButtonBase, bInherited=true, widget=null}
 ]
}
