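# Editor widget Blueprint for a tab button in the quest editor's quest item definition component UI, derived from the native FortQuestItemDefComponentTabButton class.
# The layout is a single small-font CommonTextBlock, Text_ComponentName, which fills the inherited widget slot and holds placeholder text.
# The inherited InputActionWidget is left unplaced.
# The event graph and variable list are empty, so the Blueprint supplies only appearance and adds no logic of its own.
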WidgetBlueprint /Game/NeverCook/EditorUtilities/QuestTools/QuestEditor/QuestItemDefComponentUI/FortQuestItemDefComponentTabButton
graphs [EventGraph {links [], nodes []}]
variables []
widgets {
 info {
  inheritedWidgetCount=2
  parentClass=/Script/FortniteEditor.FortQuestItemDefComponentTabButton
  rootWidgetClass=&CommonTextBlock
  widgetCount=2
 }
 widgets [
  Text_ComponentName {
   widgetClassPath=&CommonTextBlock
   bInherited=true
   widget=Text_ComponentName
   properties {
    text=Text Block
    colorAndOpacity {colorUseRule=UseColor_Specified, specifiedColor=<rgba 1 0.9451 0.9768 1>}
    font {
     bForceMonospaced=false
     bMaterialIsStencil=false
     fontMaterial=null
     fontObject=&NotoSans
     letterSpacing=0
     monospacedWidth=1
     size=10
     skewAmount=0
     typefaceFontName=NotoSans-Regular
     outlineSettings {
      bApplyOutlineToDropShadows=false
      bMiteredCorners=false
      bSeparateFillAlpha=false
      outlineMaterial=null
      outlineSize=0
      outlineColor=<rgba 0 0 0 1>
     }
    }
    shadowOffset=<0 0>
   }
  }
  InputActionWidget {widgetClassPath=&CommonActionWidget, bInherited=true, widget=null}
 ]
}
