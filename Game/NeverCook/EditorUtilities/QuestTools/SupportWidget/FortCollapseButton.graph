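# Editor utility widget blueprint based on FortCollapseButton, providing a collapse button for the NeverCook quest tools' support widgets.
# Its root is a lazy-loaded image, Image_CollapseStatus, alongside an inherited input action widget.
# An OnVisibilityStateBrushChanged override copies the supplied brush onto that image, so the icon reflects the collapse state.
# The event graph holds only empty PreConstruct, Construct and Tick events.
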WidgetBlueprint /Game/NeverCook/EditorUtilities/QuestTools/SupportWidget/FortCollapseButton
graphs [
 EventGraph {nodes [
   Event PreConstruct()
   Event Construct()
   Event Tick()
  ]}
 OnVisibilityStateBrushChanged {nodes [
   n1 = OnVisibilityStateBrushChanged!() -> (then, Brush:"Slate Brush Structure (by ref)") {is=title K2Node_FunctionEntry}
   > SetBrush(self=Get Image_CollapseStatus() -> (Image_CollapseStatus:Fort Lazy Image Object Reference), InBrush=@n1)
   > Return Node()
  ]}
]
variables []
widgets {
 info {
  inheritedWidgetCount=1
  parentClass=/Script/FortniteEditor.FortCollapseButton
  rootWidgetClass=&FortLazyImage
  widgetCount=2
 }
 widgets [
  Image_CollapseStatus {widgetClassPath=&FortLazyImage, bIsVariable=true, widget=Image_CollapseStatus}
  InputActionWidget {widgetClassPath=&CommonActionWidget, bInherited=true, widget=null}
 ]
}
